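# Performance-test material that procedurally animates an instanced static-mesh swarm bug through World Position Offset.
# Custom primitive data supplies state pair, blend start time, random phase and aggressive stance, crossfading idle, walk, melee, falling, death, flower and jump.
# Vertex-color and local-position masks isolate body parts for sine-driven offsets and pivot rotations that also rotate normals.
# Static switches substitute preview states, debug state colors or, apparently, Niagara-style dynamic parameters for VFX.
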Material /ProtoSwarm/NPCs/Swarm/TidalCrane_Swarm_WPOTest/Materials/PerfTest/M_SwarmerWPO_PerfTest
graph {
 properties {
  bAutomaticallySetUsageInEditor=false
  bTangentSpaceNormal=false
  bUseMaterialAttributes=true
  bUsedWithInstancedStaticMeshes=true
  previewMesh=/ProtoSwarm/NPCs/Swarm/TidalCrane_Swarm_WPOTest/Meshes/SM_BugProto_TA2
  usedShadingModels=Default Lit
 }
 nodes [
  n1 = DynamicParameter() {defaultValue=<rgba 0 0 1 0>, paramNames [Stance_Agressive, Idle, Walk, AttackMelee]}
  NamedRerouteDeclaration(None=@n1.Stance_Agressive) {name=VFX_Agressive, nodeColor=<rgba 0 0.25 1 1>}
  NamedRerouteDeclaration(None=@n1.Idle) {name=VFX_Idle, nodeColor=<rgba 0 0.25 1 1>}
  NamedRerouteDeclaration(None=@n1.Walk) {name=VFX_Walk, nodeColor=<rgba 0 0.25 1 1>}
  NamedRerouteDeclaration(None=@n1.AttackMelee) {name=VFX_MeleeAttack, nodeColor=<rgba 0 0.25 1 1>}
  n2 = DynamicParameter() {parameterIndex=1, defaultValue=<rgba 0 0 0 0>, paramNames [Falling,Death,Flower,Jump]}
  NamedRerouteDeclaration(None=@n2.Falling) {name=VFX_Falling, nodeColor=<rgba 0 0.25 1 1>}
  NamedRerouteDeclaration(None=@n2.Death) {name=VFX_Death, nodeColor=<rgba 0 0.25 1 1>}
  NamedRerouteDeclaration(None=@n2.Flower) {name=VFX_Flower, nodeColor=<rgba 0 0.25 1 1>}
  n3 = ScalarParameter() {
   defaultValue=2
   desc="0 - Idle\r\n1 - Walk\r\n2 - MeleeAttack\r\n3- Falling\r\n4 - Death\r\n5 - Flower\r\n6 - Jump"
   group=WPO
   parameterName="___Preview Anim (Index)"
   sortPriority=-10
  }
  n4 = ScalarParameter() {
   desc="0 - Idle\r\n1 - Walk\r\n2 - MeleeAttack\r\n3- Falling\r\n4 - Death\r\n5 - Flower\r\n6 - Jump"
   group=WPO
   parameterName=___Preview AnimA
   sortPriority=-5
  }
  n5 = NamedRerouteUsage()
  n6 = NamedRerouteDeclaration(None=StaticBoolParameter() {defaultValue=true, group=WPO, parameterName="• AnimBlendPreview", sortPriority=-5}) {name=DebugAnim, nodeColor=<rgba 1 0.0469 0 1>}
  n7 = StaticBoolParameter() {group=WPO, parameterName="• SingleAnimPreview", sortPriority=-10}
  NamedRerouteDeclaration(None=StaticSwitch(True=@n3, False=StaticSwitch(True=@n4, False=@n5, Value=@n6), Value=@n7)) {name=StateA_Value, nodeColor=<rgba 1 0.5625 0 1>}
  n8 = ScalarParameter() {
   defaultValue=1
   desc="0 - Idle\r\n1 - Walk\r\n2 - MeleeAttack\r\n3- Falling\r\n4 - Death\r\n5 - Flower\r\n6 - Jump"
   group=WPO
   parameterName=___Preview AnimB
   sortPriority=-5
  }
  n9 = NamedRerouteUsage()
  NamedRerouteDeclaration(None=StaticSwitch(True=@n3, False=StaticSwitch(True=@n8, False=@n9, Value=@n6), Value=@n7)) {name=StateB_Value, nodeColor=<rgba 0.2422 0 1 1>}
  n10 = NamedRerouteUsage()
  n11 = Constant4Vector() {desc="0 - Idle\r\n1 - Walk\r\n2 - MeleeAttack\r\n3- Falling\r\n4 - Death\r\n5 - Flower\r\n6 - Jump", constant=<rgba 0 1 2 3>}
  NamedRerouteDeclaration(None=Multiply(A=Step(Y=@n11.RGBA, X=@n10), B=Step(Y=@n10, X=@n11.RGBA))) {name=IndexA_Validation 1, nodeColor=<rgba 0.0781 1 0 1>}
  n12 = NamedRerouteUsage()
  n13 = Constant4Vector() {desc="0 - Idle\r\n1 - Walk\r\n2 - MeleeAttack\r\n3- Falling\r\n4 - Death\r\n5 - Flower\r\n6 - Jump", constant=<rgba 4 5 6 7>}
  NamedRerouteDeclaration(None=Multiply(A=Step(Y=@n13.RGBA, X=@n12), B=Step(Y=@n12, X=@n13.RGBA))) {name=IndexA_Validation 2, nodeColor=<rgba 0.0781 1 0 1>}
  n14 = NamedRerouteUsage()
  NamedRerouteDeclaration(None=Multiply(A=Step(Y=@n11.RGBA, X=@n14), B=Step(Y=@n14, X=@n11.RGBA))) {name=IndexB_Validation 1, nodeColor=<rgba 0.4429 0 1 1>}
  n15 = NamedRerouteUsage()
  NamedRerouteDeclaration(None=Multiply(A=Step(Y=@n13.RGBA, X=@n15), B=Step(Y=@n15, X=@n13.RGBA))) {name=IndexB_Validation 2, nodeColor=<rgba 0.4429 0 1 1>}
  NamedRerouteDeclaration(None=@n2.Jump) {name=VFX_Jump, nodeColor=<rgba 0 0.25 1 1>}
  n16 = Divide(A=Constant() {desc="Curve Index (0.5,1.5,2.5...)", r=0.5}, B=NamedRerouteUsage())
  n17 = Add(A=AppendVector(A=NamedRerouteUsage(), B=ComponentMask(None=@n16) {g=true}), B=Constant2Vector() {r=0.25})
  NamedRerouteDeclaration(None=TextureSample(UVs=Frac(None=@n17), Tex=NamedRerouteUsage()) {samplerType=SAMPLERTYPE_LinearColor, texture=/ProtoSwarm/NPCs/Swarm/TidalCrane_Swarm_WPOTest/Textures/CA_SwarmerWPOCurves}.R) {name=Cos_1x, nodeColor=<rgba 0 1 0.415 1>}
  n18 = Multiply(A=Constant3Vector() {desc=50, constant=<rgba 142 134 0 0>}, B=NamedRerouteUsage())
  n19 = Add(
   A=Multiply(A=Constant3Vector() {constant=<rgba 134 30 0 0>}, B=NamedRerouteUsage())
   B=Multiply(A=Constant3Vector() {constant=<rgba 73 -175 0 0>}, B=NamedRerouteUsage())
  )
  NamedRerouteDeclaration(None=Multiply(A=Add(A=@n18, B=@n19), B=NamedRerouteUsage())) {name=Legs_Pivot Point_A, nodeColor=<rgba 1 0.7734 0 1>}
  n20 = TextureObject() {samplerType=SAMPLERTYPE_LinearColor, texture=&CA_SwarmerWPOCurves}
  NamedRerouteDeclaration(None=TextureProperty(Tex=NamedRerouteDeclaration(None=@n20) {name=CA_Tex, nodeColor=<rgba 0.4531 0 1 1>})) {name=CA_Size, nodeColor=<rgba 1 0.7734 0 1>}
  n21 = VectorParameter() {bUseCustomPrimitiveData=true, parameterName=AnimationBlend, primitiveDataIndex=10}
  NamedRerouteDeclaration(None=@n21.A) {name=PD_StateB, nodeColor=<rgba 0.4297 0 1 1>}
  NamedRerouteDeclaration(None=@n21.B) {name=PD_StateA, nodeColor=<rgba 0 1 0.1094 1>}
  n22 = LocalPosition() {localOrigin=InstancePreSkinning}
  n23 = ShaderStageSwitch(PixelShader=VertexInterpolator(VS=@n22.XYZ).PS, VertexShader=@n22.XYZ)
  n24 = ComponentMask(None=@n23) {r=true}
  NamedRerouteDeclaration(None=Step(X=@n24)) {name=M_Left/Right, nodeColor=<rgba 0 0.0625 1 1>}
  n25 = LinearInterpolate(A=NamedRerouteUsage(), B=NamedRerouteUsage(), Alpha=NamedRerouteUsage())
  n26 = MaterialFunctionCall(
   R=ScalarParameter() {defaultValue=0.5, group=WPO, parameterName=BlendDur_Idle}
   G=ScalarParameter() {defaultValue=0.2, group=WPO, parameterName=BlendDur_Walk}
   B=ScalarParameter() {defaultValue=0.2, group=WPO, parameterName=BlendDur_MeleeAttack}
   A=ScalarParameter() {defaultValue=0.5, group=WPO, parameterName=BlendDur_Falling}
  ) {materialFunction=&AppendMany}
  n27 = LinearInterpolate(A=NamedRerouteUsage(), B=NamedRerouteUsage(), Alpha=NamedRerouteUsage())
  n28 = MaterialFunctionCall(
   R=ScalarParameter() {defaultValue=0.5, group=WPO, parameterName=BlendDur_Death}
   G=ScalarParameter() {defaultValue=0.5, group=WPO, parameterName=BlendDur_Flower}
   B=ScalarParameter() {defaultValue=0.5, group=WPO, parameterName=BlendDur_Jump}
  ) {materialFunction=&AppendMany}
  n29 = MaterialFunctionCall(Float4=Add(A=Multiply(A=@n25, B=@n26.RGBA), B=Multiply(A=@n27, B=@n28.RGBA))) {materialFunction=&BreakOutFloat4Components}
  n30 = Add(A=@n29.R, B=@n29.R)
  n31 = Add(A=@n29.B, B=@n29.B)
  NamedRerouteDeclaration(None=Max(A=Add(A=@n30, B=@n31)) {constB=0.01}) {name=BlendDur, nodeColor=<rgba 1 0.5859 0 1>}
  NamedRerouteDeclaration(None=Divide(A=Time(), B=ScalarParameter() {defaultValue=0.5, group=WPO, parameterName=Attack_Duration, primitiveDataIndex=1, sortPriority=1})) {name=Time_No Loop, nodeColor=<rgba 0.6406 0 1 1>}
  n32 = Time()
  n33 = Constant() {r=5}
  n34 = Divide(A=@n32, B=@n33) {constB=5}
  NamedRerouteDeclaration(None=Step(Y=Frac(None=@n34)) {constX=0.5}) {name=Preview_BlendAorB, nodeColor=<rgba 0 0.7422 1 1>}
  NamedRerouteDeclaration(None=SphereMask(A=@n23, B=Constant3Vector() {constant=<rgba 0 153 100 0>}) {attenuationRadius=130}) {name=M_HeadSphere, nodeColor=<rgba 1 0 0.7031 1>}
  n35 = VertexColor()
  n36 = Multiply(A=@n35.B, B=@n35.B)
  n37 = Step(X=Subtract(A=@n36, B=NamedRerouteUsage())) {constY=0.1}
  n38 = NamedRerouteUsage()
  n39 = Subtract(
   A=Subtract(A=@n35.R, B=NamedRerouteDeclaration(None=@n37) {name=M_Wings, nodeColor=<rgba 0 1 0.9766 1>})
   B=NamedRerouteDeclaration(None=Multiply(A=@n36, B=@n38)) {name=M_Mouth, nodeColor=<rgba 0.3828 1 0 1>}
  )
  n40 = NamedRerouteUsage()
  NamedRerouteDeclaration(None=Saturate(None=Subtract(A=@n39, B=@n40))) {name=M_Abdomen, nodeColor=<rgba 1 0.6562 0 1>}
  n41 = NamedRerouteUsage()
  n42 = Constant() {r=-75}
  n43 = Divide(A=NamedRerouteUsage(), B=Constant() {r=-15})
  n44 = LinearInterpolate(A=Constant() {r=-0.05}, B=Constant() {r=-0.02}, Alpha=NamedRerouteUsage())
  n45 = Add(A=Multiply(A=Divide(A=@n41, B=@n42), B=NamedRerouteUsage()), B=Multiply(A=Add(A=@n43, B=@n44) {constB=-0.05}, B=NamedRerouteUsage()))
  n46 = LinearInterpolate(A=@n45, B=LinearInterpolate(A=@n45, Alpha=NamedRerouteUsage()) {constB=0}, Alpha=NamedRerouteUsage()) {constB=0}
  n47 = NamedRerouteUsage()
  n48 = LinearInterpolate(
   A=Multiply(A=Multiply(A=@n47) {constB=0.1}, B=NamedRerouteUsage())
   B=Multiply(A=NamedRerouteUsage()) {constB=0.21}
   Alpha=NamedRerouteUsage()
  )
  n49 = NamedRerouteUsage()
  NamedRerouteDeclaration(None=Multiply(A=LinearInterpolate(A=@n46, B=@n48, Alpha=@n49), B=NamedRerouteUsage())) {name=Rot_Legs, nodeColor=<rgba 0 1 0.2969 1>}
  n50 = Step(X=@n35.G) {constY=0.7}
  n51 = NamedRerouteDeclaration(None=Step(X=@n35.G) {constY=0.9}) {name=M_Legs_Front, nodeColor=<rgba 0 0.2969 1 1>}
  n52 = Subtract(A=OneMinus(None=@n50), B=@n51)
  n53 = NamedRerouteDeclaration(None=Step(X=@n35.G) {constY=0.1}) {name=M_Legs, nodeColor=<rgba 0.6641 0 1 1>}
  NamedRerouteDeclaration(None=Saturate(None=Multiply(A=@n52, B=@n53))) {name=M_Legs_Back, nodeColor=<rgba 0 0.2969 1 1>}
  NamedRerouteDeclaration(None=Saturate(None=Subtract(A=@n50, B=@n51))) {name=M_Legs_Middle, nodeColor=<rgba 0 0.2969 1 1>}
  n54 = NamedRerouteDeclaration(None=@n35.A) {name=M_AGradients, nodeColor=<rgba 1 0.0234 0 1>}
  n55 = Multiply(A=@n54, B=NamedRerouteUsage())
  NamedRerouteDeclaration(None=OneMinus(None=NamedRerouteDeclaration(None=@n55) {name=M_LegsGradient, nodeColor=<rgba 1 0.2521 0.4356 1>})) {name=M_TorsoGradient, nodeColor=<rgba 0.8281 0 1 1>}
  n56 = Constant3Vector() {constant=<rgba 22 25 60 0>}
  n57 = NamedRerouteUsage()
  NamedRerouteDeclaration(None=Multiply(A=Multiply(A=@n56, B=@n57), B=NamedRerouteUsage())) {name=Legs_Pivot Point_C, nodeColor=<rgba 1 0.7734 0 1>}
  n58 = NamedRerouteUsage()
  NamedRerouteDeclaration(None=Transform(None=@n58) {transformSourceType=TRANSFORMSOURCE_Local}) {name=WPO, nodeColor=<rgba 0 0.9297 1 1>}
  n59 = Add(
   A=Multiply(A=Constant3Vector() {constant=<rgba 0 1 0 0>}, B=NamedRerouteUsage())
   B=Multiply(A=Constant3Vector() {constant=<rgba 1 0 0 0>}, B=NamedRerouteUsage())
  )
  n60 = Multiply(
   A=LinearInterpolate(A=Constant3Vector() {constant=<rgba 0 0 1 0>}, B=Constant3Vector() {constant=<rgba 1 1.5 0 0>}, Alpha=NamedRerouteUsage())
   B=NamedRerouteUsage()
  )
  n61 = NamedRerouteUsage()
  NamedRerouteDeclaration(None=Multiply(A=LinearInterpolate(A=@n59, B=@n60, Alpha=@n61), B=NamedRerouteUsage())) {name=Axis_Legs, nodeColor=<rgba 0 0.3437 1 1>}
  n62 = Time()
  n63 = ScalarParameter() {defaultValue=2, group=WPO_Bake, parameterName="Anim_Speed_Primary(s)_Idle"}
  n64 = ScalarParameter() {bUseCustomPrimitiveData=true, parameterName=RandomTime, primitiveDataIndex=4}
  NamedRerouteDeclaration(None=Add(A=Divide(A=@n62, B=@n63), B=@n64)) {name=Time_Primary_Idle, nodeColor=<rgba 0.7578 1 0 1>}
  n65 = Constant3Vector() {constant=<rgba 80 75 70 0>}
  n66 = NamedRerouteUsage()
  NamedRerouteDeclaration(None=Multiply(A=Multiply(A=@n65, B=@n66), B=NamedRerouteUsage())) {name=Legs_Pivot Point_B, nodeColor=<rgba 1 0.7734 0 1>}
  n67 = LocalPosition() {includedOffsets=ExcludeOffsets, localOrigin=InstancePreSkinning}
  NamedRerouteDeclaration(None=ShaderStageSwitch(PixelShader=VertexInterpolator(VS=@n67.XYZ).PS, VertexShader=@n67.XYZ)) {name=Pos_Preskined, nodeColor=<rgba 0 0.9766 1 1>}
  n68 = NamedRerouteUsage()
  n69 = Constant3Vector() {constant=<rgba 0 -40 0 0>}
  n70 = NamedRerouteUsage()
  n71 = NamedRerouteUsage()
  NamedRerouteDeclaration(None=Multiply(A=Multiply(A=@n68, B=@n69), B=Max(A=@n70, B=@n71))) {name=Pose_Body, nodeColor=<rgba 1 0.9141 0 1>}
  n72 = Sign(None=@n24)
  n73 = Constant() {r=1}
  NamedRerouteDeclaration(None=MaterialFunctionCall(R=NamedRerouteDeclaration(None=@n72) {name="M_Left/Right_(-1,1)", nodeColor=<rgba 0 0.0625 1 1>}, G=@n73, B=@n73) {materialFunction=&AppendMany}.RGB) {name="M_Left/Right(-1,1)_Red", nodeColor=<rgba 0.5938 0 1 1>}
  n74 = NamedRerouteUsage()
  n75 = MaterialFunctionCall!(
   RotationAngle=NamedRerouteUsage()
   NormalizedRotationAxis=Normalize(VectorInput=Add(A=@n74) {constB=0, desc=SafeNormalize})
   PivotPoint=NamedRerouteUsage()
   Position=NamedRerouteUsage()
   Normal
   TangentU
   TangentV
  ) {materialFunction=/Game/Characters/Player/Male/Medium/Bodies/M_MED_Inferno/FX/Mat/RotateVertexAboutAxis}
  NamedRerouteDeclaration(None=LinearInterpolate(A=VertexNormalWS(), B=@n75.Normal, Alpha=NamedRerouteUsage())) {name=N_Rotations1, nodeColor=<rgba 0 1 0.2734 1>}
  n76 = Sine(None=Divide(A=@n32, B=@n33) {constB=5})
  n77 = NamedRerouteUsage()
  n78 = Abs(None=Subtract(A=Time(), B=@n77))
  n79 = NamedRerouteUsage()
  n80 = Multiply(A=Saturate(None=Divide(A=@n78, B=@n79)), B=Saturate(None=@n77))
  n81 = Multiply(A=@n80, B=@n80)
  n82 = Multiply(A=@n81, B=@n81)
  n83 = StaticSwitch(
   True=Saturate(None=Multiply(A=@n76) {constB=2})
   False=LinearInterpolate(A=OneMinus(None=@n82), B=@n82, Alpha=NamedRerouteUsage()) {constAlpha=0.4}
   Value=NamedRerouteUsage()
  )
  n84 = OneMinus(None=@n83)
  n85 = NamedRerouteUsage()
  n86 = NamedRerouteUsage()
  n87 = MaterialFunctionCall(Float4=Add(A=Multiply(A=@n84, B=@n85), B=Multiply(A=@n83, B=@n86))) {materialFunction=&BreakOutFloat4Components}
  NamedRerouteDeclaration(None=StaticSwitch(True=NamedRerouteUsage(), False=@n87.G, Value=NamedRerouteUsage())) {name=Anim_1_Walk, nodeColor=<rgba 0.2457 0.7088 1 1>}
  NamedRerouteDeclaration(None=Divide(A=ComponentMask(None=@n23) {b=true}) {constB=250}) {name=M_VerticalGradient, nodeColor=<rgba 1 0.4453 0 1>}
  NamedRerouteDeclaration(None=Add(A=@n58, B=NamedRerouteUsage()) {constB=0}) {name=WPO_BeforeBodyRotation, nodeColor=<rgba 0.125 0 1 1>}
  NamedRerouteDeclaration(None=@n21.R) {name=PD_GameTime, nodeColor=<rgba 0 0.9297 1 1>}
  n88 = NamedRerouteUsage()
  n89 = Multiply(A=@n54, B=NamedRerouteDeclaration(None=Multiply(A=@n35.B, B=@n88)) {name=M_Head, nodeColor=<rgba 0 1 0.3438 1>})
  n90 = NamedRerouteUsage()
  NamedRerouteDeclaration(None=Saturate(None=Subtract(A=@n89, B=@n90))) {name=M_AntlersGrad, nodeColor=<rgba 0.8047 0 1 1>}
  n91 = Multiply(A=NamedRerouteUsage(), B=Constant() {r=0.02}) {constB=0}
  n92 = Constant() {r=-0.035}
  n93 = LinearInterpolate(
   A=LinearInterpolate(A=Constant() {r=-0.005}, B=Add(A=@n91, B=@n92), Alpha=NamedRerouteUsage())
   B=LinearInterpolate(A=Constant() {r=0.005}, B=Constant() {r=0.01}, Alpha=NamedRerouteUsage())
   Alpha=NamedRerouteUsage()
  )
  n94 = Constant() {r=0.027}
  n95 = LinearInterpolate(A=Multiply(A=NamedRerouteUsage(), B=Add(A=@n93, B=@n94)) {constB=0.01}, B=NamedRerouteUsage(), Alpha=NamedRerouteUsage())
  n96 = LinearInterpolate(A=Add(A=Multiply(A=@n95) {constB=0.1}) {constB=-0.1}, Alpha=NamedRerouteUsage()) {constA=-0.1, constB=-0.25}
  n97 = NamedRerouteUsage()
  n98 = LinearInterpolate(
   A=LinearInterpolate(A=Constant3Vector() {constant=<rgba 0 1 0 0>}, B=Constant3Vector() {constant=<rgba 0 0 1 0>}, Alpha=NamedRerouteUsage())
   B=Constant3Vector() {constant=<rgba 1 0 0 0>}
   Alpha=NamedRerouteUsage()
  ) {constB=0.2}
  n99 = NamedRerouteUsage()
  n100 = NamedRerouteUsage()
  NamedRerouteDeclaration(None=MaterialFunctionCall(
    RotationAngle=LinearInterpolate(A=@n95, B=@n96, Alpha=@n97) {constB=-0.1}
    NormalizedRotationAxis=Normalize(VectorInput=@n98)
    PivotPoint=Constant3Vector() {constant=<rgba 0 3 80 0>}
    Position=Add(A=@n99, B=@n100)
    Normal=NamedRerouteUsage()
   ) {materialFunction=&RotateVertexAboutAxis}.Normal) {name=N_Rotations2, nodeColor=<rgba 1 0.4219 0 1>}
  n101 = Multiply(A=NamedRerouteUsage(), B=ScalarParameter() {defaultValue=-30, group=WPO_Bake, parameterName=Legs Offset_Front_X})
  n102 = LinearInterpolate(
   A=LinearInterpolate(
    A=ScalarParameter() {defaultValue=-30, group=WPO_Bake, parameterName=Legs Offset_Front_Y}
    B=ScalarParameter() {defaultValue=-40, group=WPO_Bake, parameterName=Legs Offset_MIddle_Y}
    Alpha=NamedRerouteUsage()
   )
   B=ScalarParameter() {defaultValue=-10, group=WPO_Bake, parameterName=Legs Offset_Back_Y}
   Alpha=NamedRerouteUsage()
  )
  n103 = Constant()
  n104 = Multiply(A=Multiply(A=NamedRerouteUsage(), B=Constant3Vector() {constant=<rgba 0 0 50 0>}), B=NamedRerouteUsage())
  n105 = Multiply(A=NamedRerouteUsage(), B=Constant3Vector() {constant=<rgba 20 -85 50 0>})
  n106 = NamedRerouteUsage()
  n107 = Multiply(
   A=Add(
    A=MaterialFunctionCall(R=@n101, G=@n102, B=@n103) {materialFunction=&AppendMany}.RGB
    B=LinearInterpolate(A=@n104, B=@n105, Alpha=@n106)
   )
   B=NamedRerouteUsage()
  )
  n108 = Multiply(A=LocalPosition().XYZ) {constB=-0.8}
  n109 = NamedRerouteUsage()
  NamedRerouteDeclaration(None=Multiply(A=LinearInterpolate(A=@n107, B=@n108, Alpha=@n109) {constB=0}, B=NamedRerouteUsage())) {name=Pose_Legs, nodeColor=<rgba 0.8281 0 1 1>}
  NamedRerouteDeclaration(None=Add(A=NamedRerouteUsage(), B=NamedRerouteUsage())) {name=Pre-AnimPose, nodeColor=<rgba 0 1 0.5078 1>}
  n110 = Constant() {r=0.08}
  n111 = Multiply(A=NamedRerouteUsage(), B=Constant() {r=0.15}) {constB=0}
  n112 = NamedRerouteUsage()
  n113 = NamedRerouteUsage()
  n114 = NamedRerouteUsage()
  NamedRerouteDeclaration(None=Multiply(A=Add(A=@n110, B=@n111), B=LinearInterpolate(A=@n112, B=@n113, Alpha=@n114))) {name=Rot_Body_MeleeAttackCombo, nodeColor=<rgba 0 1 0.9531 1>}
  n115 = NamedRerouteUsage()
  NamedRerouteDeclaration(None=Sine(None=@n115)) {name=Timeloop_AttackComboSine1, nodeColor=<rgba 1 0.6328 0 1>}
  NamedRerouteDeclaration(None=StaticSwitch(True=NamedRerouteUsage(), False=@n21.G, Value=NamedRerouteUsage())) {name="PD_A>B/B>A", nodeColor=<rgba 1 0.75 0 1>}
  n116 = NamedRerouteUsage()
  n117 = Add(A=Add(A=NamedRerouteUsage(), B=Multiply(A=@n116) {constB=0.21}), B=Multiply(A=NamedRerouteUsage()) {constB=0.12})
  n118 = NamedRerouteUsage()
  n119 = OneMinus(None=@n118)
  n120 = NamedRerouteUsage()
  n121 = LinearInterpolate(A=@n117, B=Add(A=@n117) {constB=0.5}, Alpha=Multiply(A=LinearInterpolate(A=@n119, B=@n118, Alpha=@n120), B=NamedRerouteUsage()))
  n122 = NamedRerouteDeclaration(None=Sine(None=@n121)) {name=Walk_Sine_A, nodeColor=<rgba 1 0.75 0 1>}
  n123 = Multiply(A=Saturate(None=Power(Base=@n122) {constExponent=4}), B=Constant3Vector() {constant=<rgba 0 0 1 0>})
  n124 = Multiply(
   A=NamedRerouteDeclaration(None=Cosine(None=@n121)) {name=Walk_Cos_A, nodeColor=<rgba 1 0.853 0 1>}
   B=Constant3Vector() {constant=<rgba -1 -1 0 0>}
  )
  n125 = Constant()
  n126 = LinearInterpolate(A=LinearInterpolate(A=Constant() {r=60}, B=Constant() {r=55}, Alpha=NamedRerouteUsage()), B=Constant() {r=50}, Alpha=NamedRerouteUsage())
  n127 = LinearInterpolate(A=Constant() {r=30}, B=Constant() {r=45}, Alpha=NamedRerouteUsage())
  n128 = Multiply(
   A=Multiply(
    A=Add(A=@n123, B=@n124)
    B=MaterialFunctionCall(R=@n125, G=@n126, B=@n127) {materialFunction=&AppendMany}.RGB
   )
   B=NamedRerouteUsage()
  )
  n129 = NamedRerouteUsage()
  NamedRerouteDeclaration(None=Add(A=Multiply(A=@n128, B=@n129), B=@n75.WPO) {constB=0}) {name=Anim_LegsMain+SecondaryTorso, nodeColor=<rgba 1 0.1875 0 1>}
  NamedRerouteDeclaration(None=StaticSwitch(True=NamedRerouteUsage(), False=@n87.R, Value=NamedRerouteUsage())) {name=Anim_0_Idle, nodeColor=<rgba 0.4809 0.1848 1 1>}
  n130 = Constant() {r=1}
  n131 = ScalarParameter() {bUseCustomPrimitiveData=true, parameterName=Stance_GameTime, primitiveDataIndex=6}
  n132 = Abs(None=Subtract(A=Time(), B=@n131))
  n133 = ScalarParameter() {defaultValue=0.1, group=WPO, parameterName=Stance_BlendDuration, primitiveDataIndex=1, sortPriority=1}
  n134 = Multiply(A=Saturate(None=Divide(A=@n132, B=@n133)), B=Saturate(None=@n131))
  n135 = Multiply(A=@n134, B=@n134)
  n136 = Multiply(A=@n135, B=@n135)
  n137 = LinearInterpolate(A=@n136, B=OneMinus(None=@n136), Alpha=ScalarParameter() {bUseCustomPrimitiveData=true, parameterName=Stance_A to B, primitiveDataIndex=5}) {constAlpha=0.4}
  NamedRerouteDeclaration(None=StaticSwitch(
    True=NamedRerouteUsage()
    False=StaticSwitchParameter(True=@n130, False=@n137) {group=WPO, parameterName="• Preview_AgressiveStance", sortPriority=-50}
    Value=NamedRerouteUsage()
   )) {name=Anim_Stance_Agressive, nodeColor=<rgba 0.0547 0 1 1>}
  n138 = ScalarParameter() {defaultValue=0.7, group=WPO_Bake, parameterName="Anim_Speed_Primary(s)_Walk"}
  NamedRerouteDeclaration(None=Add(A=Divide(A=@n62, B=@n138), B=@n64)) {name=Time_Primary_Walk, nodeColor=<rgba 0.7578 1 0 1>}
  n139 = NamedRerouteUsage()
  n140 = LinearInterpolate(A=@n139, B=NamedRerouteUsage(), Alpha=NamedRerouteUsage())
  n141 = NamedRerouteUsage()
  NamedRerouteDeclaration(None=LinearInterpolate(A=LinearInterpolate(A=@n139, B=@n140, Alpha=@n141) {constB=0}, B=NamedRerouteUsage(), Alpha=NamedRerouteUsage())) {name=Legs_PivotPoint, nodeColor=<rgba 0.2891 1 0 1>}
  NamedRerouteDeclaration(None=Sine(None=Add(A=@n115) {constB=-0.05})) {name=Timeloop_AttackComboSine1_Delay, nodeColor=<rgba 1 0.6328 0 1>}
  NamedRerouteDeclaration(None=StaticBoolParameter() {group=WPO, parameterName=Is VFX, sortPriority=-10}) {name=Is_VFX, nodeColor=<rgba 1 0.0703 0 1>}
  n142 = NamedRerouteUsage()
  n143 = Constant() {r=0.02}
  n144 = LinearInterpolate(A=Constant() {r=0.65}, B=Add(A=Multiply(A=@n142, B=@n143), B=Constant() {r=0.45}), Alpha=NamedRerouteUsage())
  n145 = Subtract(A=NamedRerouteUsage(), B=@n144)
  n146 = Subtract(A=NamedRerouteUsage(), B=@n144)
  NamedRerouteDeclaration(None=LinearInterpolate(A=Sine(None=@n145), B=Sine(None=@n146), Alpha=NamedRerouteUsage())) {name=Sine_RotTorso_Main, nodeColor=<rgba 0.7578 0 1 1>}
  NamedRerouteDeclaration(None=StaticSwitch(True=NamedRerouteUsage(), False=@n87.B, Value=NamedRerouteUsage())) {name=Anim_2_MeleeAttack, nodeColor=<rgba 1 0.0722 0.2096 1>}
  NamedRerouteDeclaration(None=StaticSwitch(True=NamedRerouteUsage(), False=@n87.A, Value=NamedRerouteUsage())) {name=Anim_3_Falling, nodeColor=<rgba 0.8498 1 0.2318 1>}
  n147 = NamedRerouteUsage()
  n148 = NamedRerouteUsage()
  n149 = MaterialFunctionCall(Float4=Add(A=Multiply(A=@n84, B=@n147), B=Multiply(A=@n83, B=@n148))) {materialFunction=&BreakOutFloat4Components}
  NamedRerouteDeclaration(None=StaticSwitch(True=NamedRerouteUsage(), False=@n149.R, Value=NamedRerouteUsage())) {name=Anim_4_Death, nodeColor=<rgba 0.0576 0.0712 0.0156 1>}
  NamedRerouteDeclaration(None=StaticSwitch(True=NamedRerouteUsage(), False=@n149.G, Value=NamedRerouteUsage())) {name=Anim_5_Flower, nodeColor=<rgba 0.128 0.0135 0.2274 1>}
  NamedRerouteDeclaration(None=StaticSwitch(True=NamedRerouteUsage(), False=@n149.B, Value=NamedRerouteUsage())) {name=Anim_6_Jump, nodeColor=<rgba 0.358 1 0.5501 1>}
  n150 = LinearInterpolate(
   A=LinearInterpolate(A=NamedRerouteUsage(), B=Constant3Vector() {constant=<rgba 0 0.9387 1 1>}, Alpha=NamedRerouteUsage())
   B=Constant3Vector() {constant=<rgba 1 0.0232 0 1>}
   Alpha=NamedRerouteUsage()
  )
  n151 = Constant3Vector() {constant=<rgba 0.0252 0 1 1>}
  n152 = NamedRerouteUsage()
  n153 = LinearInterpolate(
   A=LinearInterpolate(A=LinearInterpolate(A=@n150, B=@n151, Alpha=@n152), Alpha=NamedRerouteUsage()) {constB=0}
   B=Constant3Vector() {constant=<rgba 1 0 0.8632 1>}
   Alpha=NamedRerouteUsage()
  )
  n154 = Constant3Vector() {constant=<rgba 0.0976 1 0 1>}
  n155 = NamedRerouteUsage()
  n156 = Constant()
  SetMaterialAttributes!(
   MaterialAttributes
   Base Color=StaticSwitchParameter(
    True=LinearInterpolate(A=@n153, B=@n154, Alpha=@n155)
    False=VectorParameter() {group=WPO_Bake, parameterName=Base Color, defaultValue=<rgba 0.0789 0.0649 0.1267 1>}.RGB
   ) {group=WPO, parameterName=Debug_AnimStateColors, sortPriority=-1}
   Metallic=@n156
   Specular=@n156
   Roughness=@n156
   Emissive Color
   World Position Offset=NamedRerouteUsage()
   Normal=NamedRerouteUsage()
  ) {attributeSetTypes [
    #1
    #2
    #3
    #4
    #5
    #6
    #7
   ]}
 ]
}
